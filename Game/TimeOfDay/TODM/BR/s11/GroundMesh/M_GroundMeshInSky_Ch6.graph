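# Unlit sky material for a ground mesh in the time-of-day sky, fading a tinted ground texture into sky color toward the horizon.
# A tiled, partly desaturated texture gets morning, day, evening and night tints, selected by Color_Time_of_Day and brightened.
# Grazing view angles blend it into a Sky Light environment sample along a vertically mirrored view direction, darkened by MF_Storm_DarkeningCircle for storms.
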Material /Game/TimeOfDay/TODM/BR/s11/GroundMesh/M_GroundMeshInSky_Ch6
graph {
 properties {bIsSky=true, bTangentSpaceNormal=false, pixelDepthOffsetMode=PDOM_Legacy, shadingModel=MSM_Unlit, usedShadingModels=Unlit}
 nodes [
  n1 = TextureCoordinate()
  n2 = ScalarParameter() {defaultValue=60, parameterName=tile}
  n3 = Desaturation(
   None=TextureSample(UVs=Multiply(A=@n1, B=@n2)) {texture=&/T_ReflectionGround}.RGB
   Fraction=Constant() {r=0.7}
  )
  n4 = Multiply(A=@n3, B=VectorParameter() {parameterName=ColorMorning, defaultValue=<rgba 0.5 0.3864 0.3201 1>}.RGB)
  n5 = Multiply(A=@n3, B=VectorParameter() {parameterName=ColorDay, defaultValue=<rgba 0.6562 0.3849 0.2563 1>}.RGB)
  n6 = Multiply(A=@n3, B=VectorParameter() {parameterName=ColorEvening, defaultValue=<rgba 0.5573 0.2542 0.2375 1>}.RGB)
  n7 = Multiply(A=@n3, B=VectorParameter() {parameterName=ColorNight, defaultValue=<rgba 0.2497 0.2715 0.4062 1>}.RGB)
  n8 = Multiply(A=CameraVectorWS(), B=Constant3Vector() {constant=<rgba -1 -1 1 0>})
  n9 = Constant()
  n10 = Abs(None=ComponentMask(None=CameraVectorWS()) {b=true})
  LinearInterpolate(
   A=Multiply(A=MaterialFunctionCall(Color_Morning=@n4, Color_Day=@n5, Color_Evening=@n6, Color_Night=@n7) {materialFunction=&Color_Time_of_Day}.Result) {constB=2}
   B=MaterialFunctionCall!(
    Color=SkyLightEnvMapSample(Direction=@n8, Roughness=@n9)
    Storm Color=VectorParameter() {parameterName=Storm Darkening, defaultValue=<rgba 0.1163 0.2115 0.3333 1>}.RGB
    Height=Constant() {r=-0.08}
   ) {materialFunction=/Game/TimeOfDay/CloudsAndStorm/Materials/MaterialFunctions/MF_Storm_DarkeningCircle}.Result
   Alpha=Power(Base=OneMinus(None=@n10)) {constExponent=8}
  )
 ]
}
